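# Niagara system for the Sienna male emote that emits radial magenta electricity with glowing light.
# Two ribbon emitters burst arcs on randomized cylinder points, split by loop or spawn group, with curl noise bending the flickering set.
# A computed Scale Pos lerps from the system origin to each point over life for outward expansion.
# Lightning sprites add burst and continuous crackle, while sprite-plus-light emitters supply a brief flash and a waveform-pulsed glow.
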NiagaraSystem /BRCosmetics/Effects/Fort_Effects/Emotes/Sienna/NS_Sienna_Male_EmoteRadialElectricity
dependencies {usedDataInterfaces [&NiagaraDataInterfaceCurve], usedDynamicInputs [], usedModules [&SystemState]}
emitters {
 LargeArcs {
  emitterUpdateScript [
   &EmitterState(
    Life Cycle Mode=::NewEnumerator1
    Loop Behavior=::NewEnumerator2
    Loop Count:NiagaraInt32=&RandomRangeInt(Minimum=3, Maximum=3)
    Loop Duration=0.02
    Recalculate Duration Each Loop=true
    UseLoopDelay=false
   )
   &SpawnBurst_Instantaneous(Spawn Count=8, Use Spawn Probability=false, UseLoopCountLimit=false)
   /SetVariables(
    Emitter.Radom Height:NiagaraFloat=&RandomRangeFloat(Minimum=60, Maximum=70)
    Emitter.Random Roll:NiagaraFloat=&RandomRangeFloat(Minimum=-0.1, Maximum=0.1)
    Emitter.Random Yaw:NiagaraFloat=&RandomRangeFloat(Minimum=0.1, Maximum=-0.1)
    Emitter.Random pitch:NiagaraFloat=&RandomRangeFloat(Minimum=0, Maximum=1)
    Emitter.random Float:NiagaraFloat=&RandomRangeFloat(Minimum=1, Maximum=0)
   ) {isSetParameters=true, module=SetVariables_#1}
  ]
  particleSpawnScript [
   /Game/Effects/Niagara/Modules/Partition/FN_ParticlePartition_v2/FN_Particle_Partition(Enable Ribbon ID=true, Seed:NiagaraInt32=$Emitter.Seed, Shuffle Partition ID=true) {enabled=false}
   &InitializeRibbon(
    Color=<rgba 6 0 6 1>
    Lifetime=&UniformRangedFloat(Minimum=0.1, Maximum=0.2)
    Ribbon Width=10
    UsePositionOffset=false
    Write Material Random=true
   )
   /SetVariables(
    Particles.RibbonID=/Niagara/DynamicInputs/NiagaraID/MakeNiagaraIDFromInt(Index:NiagaraInt32=$Emitter.LoopCount, Acquire Tag:NiagaraInt32=0)
    Particles.link order:NiagaraFloat=&MakeFloatFromInt(Int=&ReturnParticleIDAttributes())
   ) {isSetParameters=true, module=SetVariables_#2}
   &CylinderLocation(
    Clamp UV Range=false
    Cylinder Distribution:ENiagara_CylinderMode=::NewEnumerator2
    Cylinder Height:NiagaraFloat=$Emitter.Radom Height
    Cylinder Height Midpoint=0
    Cylinder Origin:NiagaraPosition=/Niagara/DynamicInputs/Helpers/SimulationPosition()
    Cylinder Radius=&RandomRangeFloat(Minimum=5, Maximum=7)
    MaskSpawnGroup=false
    Pitch:NiagaraFloat=$Emitter.Random Yaw
    Radius Position:NiagaraFloat=&FloatFromCurve(
     FloatCurve=DataInterface() {propertyValues {Curve {keys [
         {
          arriveTangent=3.759
          arriveTangentWeight=0.1067
          interpMode=RCIM_Cubic
          leaveTangent=3.759
          leaveTangentWeight=0.3202
          tangentMode=RCTM_User
          time=0
          value=0
         }
         {interpMode=RCIM_Cubic, tangentMode=RCTM_User, time=0.8, value=1}
        ]}}}
     CurveIndex=&ReturnNormalizedExecIndex()
    )
    Roll:NiagaraFloat=$Emitter.Random Roll
    Rotation Coordinate Space:ENiagaraCoordinateSpace=::World
    Rotation Mode:ENiagara_RotationMode=::NewEnumerator2
    U Position:NiagaraFloat=&Add_Float(
     A=&Add_Float(A=&Add_Float(A=&ReturnNormalizedExecIndex(), B=0.35), B=$Emitter.random Float)
     B=&RandomRangeFloat(Minimum=-0.2, Maximum=0.2)
    )
    UseOwnerScale=true
    V Position:NiagaraFloat=&ReturnNormalizedExecIndex()
    Yaw:NiagaraFloat=&Add_Float(A=$Emitter.Random pitch, B=&RandomRangeFloat(Minimum=-0.05, Maximum=0.05))
   )
   &FN_SetSpawnGroup(Normalize Spawn Group=true, Particles Per Group=8, Pattern=::NewEnumerator0, Set Ribbon ID and Ribbon Link Order=true)
  ]
  particleUpdateScript [
   &ParticleState()
   &ScaleColor(
    Scale Alpha=&FloatFromCurve(FloatCurve=DataInterface() {propertyValues {Curve {keys [
         {arriveTangent=-1.2679, interpMode=RCIM_Cubic, time=0.7524, value=1}
         {arriveTangent=-10.4882, interpMode=RCIM_Cubic, leaveTangent=-10.4882, tangentMode=RCTM_User, time=1, value=0}
        ]}}})
    ScaleA=true
    ScaleRGB=true
   )
   &ScaleRibbonWidth(Ribbon Width Scale=&Multiply_Float(
     A=&FloatFromCurve(
      FloatCurve=DataInterface() {propertyValues {Curve {keys [<at 0 5>,<at 0.8 0>]}}}
      CurveIndex=$Particles.Normalized Spawn Group
      Scale Curve=1.1
     )
     B=1
    )) {module=ScaleRibbonWidth001}
   /SetVariables(Particles.Scale Pos:Vector3f=&Lerp_Vector(
     A=&SimulationPosition()
     B=Linked() {linkedVariable {name=Particles.Position, type=NiagaraPosition}}
     Alpha=&Multiply_Float(
      A=&FloatFromCurve(FloatCurve=DataInterface() {propertyValues {Curve {keys [
           {
            arriveTangent=1.3693
            arriveTangentWeight=0.6453
            interpMode=RCIM_Cubic
            leaveTangent=1.3693
            leaveTangentWeight=0.6453
            tangentMode=RCTM_User
            tangentWeightMode=RCTWM_WeightedBoth
            time=0
            value=0
           }
           {
            arriveTangentWeight=1.0132
            interpMode=RCIM_Cubic
            leaveTangentWeight=1.0132
            tangentMode=RCTM_User
            tangentWeightMode=RCTWM_WeightedBoth
            time=1
            value=1
           }
          ]}}})
      B=1.6
     )
    )) {isSetParameters=true}
   &CurlNoiseForce(Mask Curl Noise=true, Noise Quality / Cost=::NewEnumerator0, Noise Strength=15000, Pan Noise=false) {enabled=false}
   &SolveForcesAndVelocity(Clamp Velocity=false, Limit Acceleration=false) {enabled=false}
  ]
  rendererClasses [&NiagaraRibbonRendererProperties, &NiagaraSpriteRendererProperties]
  renderers [
   {rendererClass=&NiagaraRibbonRendererProperties}
   {rendererClass=&NiagaraSpriteRendererProperties, rendererIndex=1}
  ]
 }
 LargeArcs001 {
  emitterUpdateScript [
   &EmitterState(Life Cycle Mode=::NewEnumerator1, Loop Duration=0.075, Recalculate Duration Each Loop=true, UseLoopDelay=false)
   &SpawnBurst_Instantaneous(Spawn Count=8, Use Spawn Probability=false, UseLoopCountLimit=false)
   /SetVariables(
    Emitter.Radom Height:NiagaraFloat=&RandomRangeFloat(Minimum=50, Maximum=75)
    Emitter.random Float:NiagaraFloat=&RandomRangeFloat(Minimum=0, Maximum=1)
   ) {isSetParameters=true, module=SetVariables_#1}
  ]
  particleSpawnScript [
   &FN_Particle_Partition(Enable Ribbon ID=true, Seed:NiagaraInt32=$Emitter.Seed, Shuffle Partition ID=true) {enabled=false}
   &InitializeRibbon(
    Color=<rgba 6 0 6 1>
    Lifetime=&UniformRangedFloat(Minimum=0.2, Maximum=0.35)
    Ribbon Width=15
    UsePositionOffset=false
    Write Material Random=true
   )
   /SetVariables(
    Particles.RibbonID=&MakeNiagaraIDFromInt(Index:NiagaraInt32=$Emitter.LoopCount, Acquire Tag:NiagaraInt32=0)
    Particles.link order:NiagaraFloat=&MakeFloatFromInt(Int=&ReturnParticleIDAttributes())
   ) {isSetParameters=true, module=SetVariables_#2}
   &CylinderLocation(
    Clamp UV Range=false
    Cylinder Distribution:ENiagara_CylinderMode=::NewEnumerator2
    Cylinder Height:NiagaraFloat=$Emitter.Radom Height
    Cylinder Height Midpoint=0
    Cylinder Origin:NiagaraPosition=&SimulationPosition()
    Cylinder Radius=&RandomRangeFloat(Minimum=5, Maximum=7)
    MaskSpawnGroup=false
    Pitch=0.25
    Radius Position:NiagaraFloat=&FloatFromCurve(
     FloatCurve=DataInterface() {propertyValues {Curve {keys [
         {
          arriveTangent=3.759
          arriveTangentWeight=0.1067
          interpMode=RCIM_Cubic
          leaveTangent=3.759
          leaveTangentWeight=0.3202
          tangentMode=RCTM_User
          time=0
          value=0
         }
         {interpMode=RCIM_Cubic, tangentMode=RCTM_User, time=0.8, value=1}
        ]}}}
     CurveIndex=&ReturnNormalizedExecIndex()
    )
    Rotation Coordinate Space:ENiagaraCoordinateSpace=::World
    Rotation Mode:ENiagara_RotationMode=::NewEnumerator2
    U Position:NiagaraFloat=&Add_Float(
     A=&Add_Float(A=&Add_Float(A=&ReturnNormalizedExecIndex(), B=0.35), B=$Emitter.random Float)
     B=&RandomRangeFloat(Minimum=-0.2, Maximum=0.2)
    )
    UseOwnerScale=true
    V Position:NiagaraFloat=&ReturnNormalizedExecIndex()
    Yaw:NiagaraFloat=&Add_Float(A=$Emitter.random Float, B=&RandomRangeFloat(Minimum=-0.05, Maximum=0.05))
   )
  ]
  particleUpdateScript [
   &ParticleState()
   &ScaleColor(
    Scale Alpha=&FloatFromCurve(FloatCurve=DataInterface() {propertyValues {Curve {keys [
         {arriveTangent=8.0553, interpMode=RCIM_Cubic, leaveTangent=8.0553, tangentMode=RCTM_User, time=0, value=0}
         <at 0.2113 1 RCIM_Cubic>
         <at 0.4546 0>
         {arriveTangent=-8.0018, interpMode=RCIM_Cubic, leaveTangent=7.1228, tangentMode=RCTM_Break, time=0.5517, value=0}
         <at 0.7524 1 RCIM_Cubic>
         {arriveTangent=-10.4882, interpMode=RCIM_Cubic, leaveTangent=-10.4882, tangentMode=RCTM_User, time=1, value=0}
        ]}}})
    ScaleA=true
    ScaleRGB=true
   )
   &ScaleRibbonWidth(Ribbon Width Scale=&Multiply_Float(
     A=&FloatFromCurve(FloatCurve=DataInterface() {propertyValues {Curve {keys [
          {interpMode=RCIM_Cubic, tangentMode=RCTM_User, time=0, value=0}
          {interpMode=RCIM_Cubic, tangentMode=RCTM_User, time=0.25, value=1}
          {arriveTangentWeight=0.1491, interpMode=RCIM_Cubic, leaveTangentWeight=0.2236, tangentMode=RCTM_User, time=0.75, value=1}
          {arriveTangentWeight=0.1, interpMode=RCIM_Cubic, tangentMode=RCTM_User, time=1, value=0}
         ]}}})
     B=&FloatFromCurve(
      FloatCurve=DataInterface() {propertyValues {Curve {keys [
          {arriveTangent=6.6108, interpMode=RCIM_Cubic, leaveTangent=6.6108, tangentMode=RCTM_User, time=0.05, value=0}
          {interpMode=RCIM_Cubic, tangentMode=RCTM_User, time=0.5, value=1}
          {arriveTangent=-6.3941, interpMode=RCIM_Cubic, leaveTangent=-6.3941, tangentMode=RCTM_User, time=0.95, value=0}
         ]}}}
      CurveIndex=&ReturnNormalizedExecIndex()
     )
    )) {module=ScaleRibbonWidth001}
   &CurlNoiseForce(Noise Strength=250, Pan Noise=false)
   &SolveForcesAndVelocity(Clamp Velocity=false, Limit Acceleration=false)
   /SetVariables(Particles.Scale Pos:Vector3f=&Lerp_Vector(
     A=&SimulationPosition()
     B=Linked() {linkedVariable {name=Particles.Position, type=NiagaraPosition}}
     Alpha=&Multiply_Float(A=$Particles.NormalizedAge, B=1)
    )) {isSetParameters=true}
  ]
  rendererClasses [&NiagaraRibbonRendererProperties, &NiagaraSpriteRendererProperties]
  renderers [
   {rendererClass=&NiagaraRibbonRendererProperties}
   {rendererClass=&NiagaraSpriteRendererProperties, rendererIndex=1}
  ]
 }
 Light {
  emitterUpdateScript [
   &EmitterState(Life Cycle Mode=::NewEnumerator1, Loop Duration=2, UseLoopDelay=false)
   &SpawnBurst_Instantaneous(Age:NiagaraFloat=$Emitter.Age, Use Spawn Probability=false, UseLoopCountLimit=false)
  ]
  particleSpawnScript [&InitializeParticle(
    Color=<rgba 1 0.35 1 8>
    Lifetime=0.2
    Mass Mode=::NewEnumerator3
    Position Offset=<0 -10 0>
    Sprite Rotation Mode=::NewEnumerator2
    Sprite Size Mode=::NewEnumerator1
    Uniform Sprite Size Max=25
    Uniform Sprite Size Min=25
    UsePositionOffset=true
    Write Material Random=true
   )]
  particleUpdateScript [
   &ParticleState(
    Kill Particles When Lifetime Has Elapsed=false
    Let Infinitely Lived Particles Die When Emitter Deactivates=true
    Loop Particles Lifetime=true
   )
   &ScaleColor(
    Scale Alpha=&Waveform(
     Waveform Count=::NewEnumerator2
     "Waveform [1]"=::NewEnumerator0
     "[1] Frequency"=2
     "[1] Amplitude Min/Max"=<0 1>
     "[2] Average"=false
     "Waveform [2]"=::NewEnumerator2
     "[2] Input Position"=0
     "[2] Phase"=$Emitter.Age
     "[2] Phase Scale"=1
     "[2] Compound Frequency"=2
     "[2] Frequency"=5
     "[2] Amplitude Min/Max"=<0 1>
     "[2] Scale"=1
     "[2] Bias"=0
     "[2] Exponent"=1
     "[2] Pulse Duration"=0.1
     "[2] Clip Triangle Min/Max"=<0 0>
     "[2] Random Seed"=0
     "[3] Average"=false
     "Waveform [3]"=::NewEnumerator8
     "[3] Input Position"=0
     "[3] Phase"=$Emitter.Age
     "[3] Phase Scale"=1
     "[3] Frequency"=4
     "[3] Compound Frequency"=2
     "[3] Amplitude Min/Max"=<0 1>
     "[3] Scale"=1
     "[3] Bias"=0
     "[3] Exponent"=1
     "[3] Pulse Duration"=0.1
     "[3] Clip Triangle Min/Max"=<0 0>
     "[3] Random Seed"=0
    )
    ScaleA=true
    ScaleRGB=true
   )
  ]
  rendererClasses [&NiagaraSpriteRendererProperties, &NiagaraLightRendererProperties]
  renderers [
   {rendererClass=&NiagaraSpriteRendererProperties}
   {rendererClass=&NiagaraLightRendererProperties, rendererIndex=1}
  ]
 }
 Light001 {
  emitterUpdateScript [
   &EmitterState(Life Cycle Mode=::NewEnumerator1, Loop Behavior=::NewEnumerator1, Loop Duration=0.1, UseLoopDelay=false)
   &SpawnBurst_Instantaneous(Age:NiagaraFloat=$Emitter.Age, Use Spawn Probability=false, UseLoopCountLimit=false)
  ]
  particleSpawnScript [&InitializeParticle(
    Color=<rgba 3 1.05 3 8>
    Lifetime=0.15
    Mass Mode=::NewEnumerator3
    Position Offset=<0 -10 0>
    Sprite Rotation Mode=::NewEnumerator2
    Sprite Size Mode=::NewEnumerator1
    Uniform Sprite Size Max=25
    Uniform Sprite Size Min=25
    UsePositionOffset=true
    Write Material Random=true
   )]
  particleUpdateScript [
   &ParticleState()
   &ScaleColor(
    Scale Alpha=&FloatFromCurve(FloatCurve=DataInterface() {propertyValues {Curve {keys [{interpMode=RCIM_Cubic, tangentMode=RCTM_User, time=0.25, value=1}, {tangentMode=RCTM_User, time=1, value=0}]}}})
    ScaleA=true
    ScaleRGB=true
   )
  ]
  rendererClasses [&NiagaraSpriteRendererProperties, &NiagaraLightRendererProperties]
  renderers [
   {rendererClass=&NiagaraSpriteRendererProperties}
   {rendererClass=&NiagaraLightRendererProperties, rendererIndex=1}
  ]
 }
 Lightning_on_ring {
  emitterUpdateScript [
   &EmitterState()
   &SpawnBurst_Instantaneous(
    Spawn Count=&RandomRangeInt(Minimum=2, Maximum=3)
    Use Spawn Probability=false
    UseLoopCountLimit=false
   )
   &SpawnRate(SpawnRate=30, Use Spawn Probability=false)
  ]
  particleSpawnScript [
   &InitializeParticle(
    Color=<rgba 3 0 2.2769 1>
    Lifetime Max=0.18
    Lifetime Min=0.1
    Lifetime Mode=::NewEnumerator1
    Sprite Size Max=<50 50>
    Sprite Size Min=<25 15>
    Sprite Size Mode=::NewEnumerator2
    Sprite UV Mode=::NewEnumerator1
    UsePositionOffset=true
    Write Material Random=true
   )
   &DynamicMaterialParameters(
    Index 0 Param 1=&RandomRangeFloat(Minimum=1, Maximum=500)
    Index 0 Param 2=&RandomRangeFloat(Minimum=1, Maximum=2)
    Module.Param0WriteEnabled=true
    Module.Param1WriteEnabled=true
    Module.Param2WriteEnabled=true
    Module.Param3WriteEnabled=true
   )
   &AddVelocity(Velocity=&RandomVector(Vector Scale=50))
   &ShapeLocation(MaskSpawnGroup=false, Override Seed=false, Sphere Radius=0, UseOwnerScale=true)
  ]
  particleUpdateScript [
   &SolveForcesAndVelocity(Clamp Velocity=false, Limit Acceleration=false)
   &ParticleState()
   &ScaleSpriteSize(
    Uniform Curve Scale=0.55
    Uniform Curve Sprite Scale=DataInterface() {propertyValues {Curve {keys [{tangentMode=RCTM_User, time=0, value=0.8}, {tangentMode=RCTM_User, time=1, value=1}]}}}
   )
  ]
  rendererClasses [&NiagaraSpriteRendererProperties]
  renderers [{rendererClass=&NiagaraSpriteRendererProperties}]
 }
 Lightning_on_ring002 {
  emitterUpdateScript [
   &EmitterState(Life Cycle Mode=::NewEnumerator1, Loop Behavior=::NewEnumerator1, Loop Duration=1, UseLoopDelay=false)
   &SpawnBurst_Instantaneous(
    Spawn Count=&RandomRangeInt(Minimum=4, Maximum=6)
    Use Spawn Probability=false
    UseLoopCountLimit=false
   )
  ]
  particleSpawnScript [
   &InitializeParticle(
    Color=<rgba 3 0 2.2769 1>
    Lifetime Max=0.18
    Lifetime Min=0.1
    Lifetime Mode=::NewEnumerator1
    Sprite Size Max=<50 50>
    Sprite Size Min=<25 15>
    Sprite Size Mode=::NewEnumerator2
    Sprite UV Mode=::NewEnumerator1
    UsePositionOffset=true
    Write Material Random=true
   )
   &DynamicMaterialParameters(
    Index 0 Param 1=&RandomRangeFloat(Minimum=1, Maximum=500)
    Index 0 Param 2=&RandomRangeFloat(Minimum=1, Maximum=2)
    Module.Param0WriteEnabled=true
    Module.Param1WriteEnabled=true
    Module.Param2WriteEnabled=true
    Module.Param3WriteEnabled=true
   )
   &AddVelocity(Velocity=&RandomVector(Vector Scale=50))
   &ShapeLocation(MaskSpawnGroup=false, Override Seed=false, Sphere Radius=0, UseOwnerScale=true)
  ]
  particleUpdateScript [
   &SolveForcesAndVelocity(Clamp Velocity=false, Limit Acceleration=false)
   &ParticleState()
   &ScaleSpriteSize(Uniform Curve Scale=2)
  ]
  rendererClasses [&NiagaraSpriteRendererProperties]
  renderers [{rendererClass=&NiagaraSpriteRendererProperties}]
 }
}
summary {
 systemName=NS_Sienna_Male_EmoteRadialElectricity
 emitters [
  {emitterName=Lightning_on_ring, rendererClasses [&NiagaraSpriteRendererProperties]}
  {emitterName=Light, rendererClasses [&NiagaraSpriteRendererProperties, &NiagaraLightRendererProperties]}
  {emitterName=Lightning_on_ring002, rendererClasses [&NiagaraSpriteRendererProperties]}
  {emitterName=LargeArcs, rendererClasses [&NiagaraRibbonRendererProperties, &NiagaraSpriteRendererProperties]}
  {emitterName=LargeArcs001, rendererClasses [&NiagaraRibbonRendererProperties, &NiagaraSpriteRendererProperties]}
  {emitterName=Light001, rendererClasses [&NiagaraSpriteRendererProperties, &NiagaraLightRendererProperties]}
 ]
 userVariables [User.RibbonIDOffset {type=NiagaraInt32}]
}
