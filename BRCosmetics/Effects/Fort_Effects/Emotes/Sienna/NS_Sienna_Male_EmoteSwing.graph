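# Niagara system for the Sienna male emote swing, drawing pink lightning ribbon trails.
# Two active emitters (a third is disabled) continuously spawn short-lived ribbon particles, ordered by age and facing along mesh-orientation Z.
# Ribbon U overrides advance with emitter age, and a dynamic material parameter animates over particle life.
# Active emitters pick color by time of day and fade both trail ends using normalized execution index.
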NiagaraSystem /BRCosmetics/Effects/Fort_Effects/Emotes/Sienna/NS_Sienna_Male_EmoteSwing
dependencies {
 usedDataInterfaces [&NiagaraDataInterfaceCurve, &NiagaraDataInterfaceColorCurve]
 usedDynamicInputs []
 usedModules [&SystemState]
}
emitters {
 Lightning_pink {
  bEnabled=false
  emitterUpdateScript [
   &EmitterState()
   &SpawnRate(SpawnRate=90, Use Spawn Probability=false)
  ]
  particleSpawnScript [
   &FN_SetSpawnGroup(Number of Spawn Groups=1, Set Ribbon ID and Ribbon Link Order=true)
   &InitializeRibbon(Color=<rgba 150 0 136.4355 25>, Lifetime=0.125, Ribbon Width=20, UsePositionOffset=false, Write Material Random=true)
   /SetVariables(Particles.RibbonU1Override:NiagaraFloat=&Add_Float(
     A=&Multiply_Float(A=$Emitter.Age, B=5)
     B=&Multiply_FloatByInt(Float=0.5, Integer=$Particles.SpawnGroup)
    )) {isSetParameters=true, module=SetVariables_#1}
   /SetVariables(Particles.RibbonU1Override2:NiagaraFloat=&Add_Float(
     A=&Multiply_Float(A=$Emitter.Age, B=7)
     B=&Multiply_FloatByInt(Float=0.5, Integer=$Particles.SpawnGroup)
    )) {isSetParameters=true, module=SetVariables_#2}
   &AddVelocity(Velocity=<0 200 0>, Velocity Speed Scale=5) {enabled=false}
   /SetVariables(
    Particles.Lightning 2 color:LinearColor=&MultiplyLinearColor(Linear Color=$Particles.Color, Scale Factor=<1 1 1 0.015>)
    Particles.ribbonWidth2:NiagaraFloat=&Multiply_Float(A=$Particles.RibbonWidth, B=2.5)
   ) {isSetParameters=true, module=SetVariables_#3}
  ]
  particleUpdateScript [
   &SolveForcesAndVelocity(Clamp Velocity=false, Limit Acceleration=false) {enabled=false}
   &ParticleState()
   &DynamicMaterialParameters(
    Index 0 Param 1=1
    Index 0 Param 2=&FloatFromCurve(
     FloatCurve=DataInterface() {propertyValues {Curve {keys [<at 0.33 0 RCIM_Cubic>, {arriveTangent=4.6706, interpMode=RCIM_Cubic, leaveTangent=4.6706, tangentMode=RCTM_User, time=1, value=1}]}}}
     Scale Curve=-1
    )
    Index 0 Param 3=1
    Index 0 Param 4=0.1
    Module.Param0WriteEnabled=true
    Module.Param1WriteEnabled=true
    Module.Param2WriteEnabled=true
    Module.Param3WriteEnabled=true
   )
   &ScaleColor(
    Scale Alpha=&Multiply_Float(
     A=&Multiply_Float(
      A=&FloatFromCurve(
       FloatCurve=DataInterface() {propertyValues {Curve {keys [
           {arriveTangent=-5.8127, interpMode=RCIM_Cubic, leaveTangent=-5.8127, tangentMode=RCTM_User, time=0.5, value=1}
           {tangentMode=RCTM_User, time=1, value=0}
          ]}}}
       CurveIndex=/Game/Effects/Niagara/DynamicInputs/Math/FN_NormalizedParticleCount()
      )
      B=&FNTimeOfDayScalar(Night Scale=0.4, Morning Scale=0.2, Evening Scale=0.4)
     )
     B=&FloatFromCurve(
      FloatCurve=DataInterface() {propertyValues {Curve {keys [
          {arriveTangent=-3.2772, interpMode=RCIM_Cubic, leaveTangent=-3.2772, tangentMode=RCTM_User, time=0, value=1}
          {arriveTangent=-3.534, interpMode=RCIM_Cubic, leaveTangent=-3.534, tangentMode=RCTM_User, time=0.375, value=0}
         ]}}}
      Scale Curve=10
     )
    )
    Scale RGB=&VectorFromFloat(Value=&FloatFromCurve(
      FloatCurve=DataInterface() {propertyValues {Curve {keys [
          {arriveTangent=-2.7165, interpMode=RCIM_Cubic, leaveTangent=-2.7165, tangentMode=RCTM_User, time=0.1408, value=0.9847}
          {arriveTangent=-0.0405, interpMode=RCIM_Cubic, leaveTangent=-0.0405, tangentMode=RCTM_User, time=1, value=0.25}
         ]}}}
      CurveIndex=&FN_NormalizedParticleCount()
      Scale Curve=2
     ))
    ScaleA=true
    ScaleRGB=true
   )
   /SetVariables(Particles.RibbonFacing=&ReturnMeshOrientationAxis(Mesh Orientation Axis=::ZAxis, Mesh Orientation Quat=$Particles.MeshOrientation)) {isSetParameters=true, module=SetVariables_#4}
   /SetVariables(Particles.RibbonLinkOrder:NiagaraFloat=$Particles.Age) {isSetParameters=true, module=SetVariables_#5}
   /SetVariables(Particles.ribbonWidth2Scaled:NiagaraFloat=&Multiply_Float(
     A=$Particles.Initial.ribbonWidth2
     B=&FloatFromCurve(FloatCurve=DataInterface() {propertyValues {Curve {keys [<at 0 1>,<at 1 1.125>]}}})
    )) {isSetParameters=true, module=SetVariables_#6}
  ]
  rendererClasses [&NiagaraRibbonRendererProperties]
  renderers [
   {rendererClass=&NiagaraRibbonRendererProperties}
   {rendererClass=&NiagaraRibbonRendererProperties, rendererIndex=1}
  ]
 }
 Lightning_pink001 {
  emitterUpdateScript [
   &EmitterState()
   &SpawnRate(SpawnRate=90, Use Spawn Probability=false)
  ]
  particleSpawnScript [
   &FN_SetSpawnGroup(Number of Spawn Groups=1, Set Ribbon ID and Ribbon Link Order=true)
   &InitializeRibbon(
    Color=&FN_Select_Color_from_Time_Of_Day(Morning=<rgba 150 0 136.4355 25>, Day=<rgba 25 0 22.7393 25>, Evening=<rgba 150 0 136.4355 25>, Night=<rgba 8 0 7.2766 25>)
    Lifetime=0.22
    Ribbon Width=20
    UsePositionOffset=false
    Write Material Random=true
   )
   /SetVariables(Particles.RibbonU1Override:NiagaraFloat=&Add_Float(
     A=&Multiply_Float(A=$Emitter.Age, B=5)
     B=&Multiply_FloatByInt(Float=0.5, Integer=$Particles.SpawnGroup)
    )) {isSetParameters=true, module=SetVariables_#1}
   /SetVariables(Particles.RibbonU1Override2:NiagaraFloat=&Add_Float(
     A=&Multiply_Float(A=$Emitter.Age, B=7)
     B=&Multiply_FloatByInt(Float=0.5, Integer=$Particles.SpawnGroup)
    )) {isSetParameters=true, module=SetVariables_#2}
   &AddVelocity(Velocity=<0 200 0>, Velocity Speed Scale=5) {enabled=false}
   /SetVariables(
    Particles.Lightning 2 color:LinearColor=&MultiplyLinearColor(Linear Color=$Particles.Color, Scale Factor=<10 10 10 0.015>)
    Particles.ribbonWidth2:NiagaraFloat=&Multiply_Float(A=$Particles.RibbonWidth, B=2.5)
   ) {isSetParameters=true, module=SetVariables_#3}
  ]
  particleUpdateScript [
   &SolveForcesAndVelocity(Clamp Velocity=false, Limit Acceleration=false) {enabled=false}
   &ParticleState()
   &DynamicMaterialParameters(
    Index 0 Param 1=1
    Index 0 Param 2=&FloatFromCurve(
     FloatCurve=DataInterface() {propertyValues {Curve {keys [<at 0.33 0 RCIM_Cubic>, {arriveTangent=4.6706, interpMode=RCIM_Cubic, leaveTangent=4.6706, tangentMode=RCTM_User, time=1, value=1}]}}}
     Scale Curve=-1
    )
    Index 0 Param 3=1
    Index 0 Param 4=0.1
    Module.Param0WriteEnabled=true
    Module.Param1WriteEnabled=true
    Module.Param2WriteEnabled=true
    Module.Param3WriteEnabled=true
   )
   &ScaleColor(
    Scale Alpha=&Multiply_Float(
     A=&Multiply_Float(
      A=&Multiply_Float(
       A=&Multiply_Float(
        A=&FloatFromCurve(
         FloatCurve=DataInterface() {propertyValues {Curve {keys [
             {arriveTangent=-5.8127, interpMode=RCIM_Cubic, leaveTangent=-5.8127, tangentMode=RCTM_User, time=0, value=1}
             {tangentMode=RCTM_User, time=0.35, value=0}
            ]}}}
         CurveIndex=&FN_NormalizedParticleCount()
        )
        B=&FNTimeOfDayScalar(Night Scale=0.4, Morning Scale=0.2, Evening Scale=0.4)
       )
       B=&FloatFromCurve(
        FloatCurve=DataInterface() {propertyValues {Curve {keys [
            {arriveTangent=-3.2772, interpMode=RCIM_Cubic, leaveTangent=-3.2772, tangentMode=RCTM_User, time=0, value=1}
            {arriveTangent=-3.534, interpMode=RCIM_Cubic, leaveTangent=-3.534, tangentMode=RCTM_User, time=0.375, value=0}
           ]}}}
        Scale Curve=10
       )
      )
      B=&FloatFromCurve(FloatCurve=DataInterface() {propertyValues {Curve {keys [
           {arriveTangentWeight=0.2853, interpMode=RCIM_Cubic, leaveTangentWeight=0.2853, time=0, value=0}
           {arriveTangentWeight=0.1667, interpMode=RCIM_Cubic, leaveTangentWeight=0.1667, time=0.1, value=1}
           {
            arriveTangentWeight=0.2502
            interpMode=RCIM_Cubic
            leaveTangentWeight=0.2502
            tangentMode=RCTM_User
            tangentWeightMode=RCTWM_WeightedBoth
            time=0.35
            value=0
           }
          ]}}})
     )
     B=&FloatFromCurve(
      FloatCurve=DataInterface() {propertyValues {Curve {keys [
          {interpMode=RCIM_Cubic, leaveTangentWeight=0.0333, tangentMode=RCTM_User, time=0, value=0}
          {arriveTangentWeight=0.1192, interpMode=RCIM_Cubic, leaveTangentWeight=0.228, tangentMode=RCTM_User, time=0.1, value=1}
          {arriveTangentWeight=0.0333, interpMode=RCIM_Cubic, tangentMode=RCTM_User, time=0.35, value=0}
         ]}}}
      CurveIndex=&FN_NormalizedParticleCount()
     )
    )
    Scale RGB=&VectorFromFloat(Value=&Multiply_Float(
      A=&FloatFromCurve(
       FloatCurve=DataInterface() {propertyValues {Curve {keys [
           {arriveTangent=-2.7165, interpMode=RCIM_Cubic, leaveTangent=-2.7165, tangentMode=RCTM_User, time=0.1408, value=0.9847}
           {arriveTangent=-0.0405, interpMode=RCIM_Cubic, leaveTangent=-0.0405, tangentMode=RCTM_User, time=0.5, value=0.25}
           <at 0.6 0>
          ]}}}
       CurveIndex=&FN_NormalizedParticleCount()
       Scale Curve=2
      )
      B=&FloatFromCurve(
       FloatCurve=DataInterface() {propertyValues {Curve {keys [
           {interpMode=RCIM_Cubic, tangentMode=RCTM_User, time=0, value=0}
           {interpMode=RCIM_Cubic, tangentMode=RCTM_User, time=0.5, value=1}
           {tangentMode=RCTM_User, time=1, value=0}
          ]}}}
       Scale Curve=2
      )
     ))
    ScaleA=true
    ScaleRGB=true
   ) {enabled=false}
   &ScaleColor(
    Scale Alpha=&FloatFromCurve(
     FloatCurve=DataInterface() {propertyValues {Curve {keys [
         {interpMode=RCIM_Cubic, tangentMode=RCTM_User, time=0, value=0}
         {
          arriveTangentWeight=0.4677
          interpMode=RCIM_Cubic
          leaveTangentWeight=0.4677
          tangentMode=RCTM_User
          tangentWeightMode=RCTWM_WeightedBoth
          time=0.5
          value=1
         }
         {tangentMode=RCTM_User, time=1, value=0}
        ]}}}
     CurveIndex=&ReturnNormalizedExecIndex()
    )
    ScaleA=true
    ScaleRGB=true
   ) {module=ScaleColor001}
   /SetVariables(Particles.RibbonFacing=&ReturnMeshOrientationAxis(Mesh Orientation Axis=::ZAxis, Mesh Orientation Quat=$Particles.MeshOrientation)) {isSetParameters=true, module=SetVariables_#4}
   /SetVariables(Particles.RibbonLinkOrder:NiagaraFloat=$Particles.Age) {isSetParameters=true, module=SetVariables_#5}
   /SetVariables(Particles.ribbonWidth2Scaled:NiagaraFloat=&Multiply_Float(
     A=$Particles.Initial.ribbonWidth2
     B=&FloatFromCurve(FloatCurve=DataInterface() {propertyValues {Curve {keys [<at 0 1>,<at 1 1.125>]}}})
    )) {isSetParameters=true, module=SetVariables_#6}
   /SetVariables(Particles.Lightning 2 color:LinearColor=&MultiplyLinearColor(
     Linear Color=$Particles.Initial.Lightning 2 color
     Scale Factor=&MakeFloat4FromVectorAndFloat(
      "Vector (RGB)"=<1 1 1>
      "Float (Alpha)"=&Multiply_Float(
       A=&FloatFromCurve(FloatCurve=DataInterface() {propertyValues {Curve {keys [
            {arriveTangent=-2.9261, interpMode=RCIM_Cubic, time=0, value=0}
            {arriveTangentWeight=0.1, interpMode=RCIM_Cubic, leaveTangentWeight=0.1667, tangentMode=RCTM_User, time=0.25, value=1}
            {arriveTangentWeight=0.2906, interpMode=RCIM_Cubic, leaveTangentWeight=0.1163, tangentMode=RCTM_User, time=0.75, value=1}
            {interpMode=RCIM_Cubic, leaveTangent=-2.6628, time=1, value=0}
           ]}}})
       B=&FloatFromCurve(
        FloatCurve=DataInterface() {propertyValues {Curve {keys [
            {arriveTangent=-2.9261, interpMode=RCIM_Cubic, time=0, value=0}
            {
             arriveTangent=7.6918
             arriveTangentWeight=0.0833
             interpMode=RCIM_Cubic
             leaveTangent=7.6918
             leaveTangentWeight=0.1667
             tangentMode=RCTM_User
             time=0.25
             value=1
            }
            {
             arriveTangent=-8.2223
             arriveTangentWeight=0.1667
             interpMode=RCIM_Cubic
             leaveTangent=-8.2223
             leaveTangentWeight=0.0833
             tangentMode=RCTM_User
             time=0.75
             value=1
            }
            {interpMode=RCIM_Cubic, leaveTangent=-2.6628, time=1, value=0}
           ]}}}
        CurveIndex=&ReturnNormalizedExecIndex()
        Scale Curve=5
       )
      )
     )
    )) {isSetParameters=true}
  ]
  rendererClasses [&NiagaraRibbonRendererProperties]
  renderers [{rendererClass=&NiagaraRibbonRendererProperties}]
 }
 Lightning_pink002 {
  emitterUpdateScript [
   &EmitterState()
   &SpawnRate(SpawnRate=90, Use Spawn Probability=false)
  ]
  particleSpawnScript [
   &FN_SetSpawnGroup(Number of Spawn Groups=1, Set Ribbon ID and Ribbon Link Order=true)
   &InitializeRibbon(
    Color=&FN_Select_Color_from_Time_Of_Day(Morning=<rgba 150 0 136.4355 25>, Day=<rgba 150 0 136.4355 25>, Evening=<rgba 150 0 136.4355 25>, Night=<rgba 10 0 9.0957 25>)
    Lifetime=0.1
    Ribbon Width=20
    UsePositionOffset=false
    Write Material Random=true
   )
   /SetVariables(Particles.RibbonU1Override:NiagaraFloat=&Add_Float(
     A=&Multiply_Float(A=$Emitter.Age, B=5)
     B=&Multiply_FloatByInt(Float=0.5, Integer=$Particles.SpawnGroup)
    )) {isSetParameters=true, module=SetVariables_#1}
   /SetVariables(Particles.RibbonU1Override2:NiagaraFloat=&Add_Float(
     A=&Multiply_Float(A=$Emitter.Age, B=7)
     B=&Multiply_FloatByInt(Float=0.5, Integer=$Particles.SpawnGroup)
    )) {isSetParameters=true, module=SetVariables_#2}
   &AddVelocity(Velocity=<0 200 0>, Velocity Speed Scale=5) {enabled=false}
   /SetVariables(
    Particles.Lightning 2 color:LinearColor=&MultiplyLinearColor(Linear Color=$Particles.Color, Scale Factor=<1 1 1 0.015>)
    Particles.ribbonWidth2:NiagaraFloat=&Multiply_Float(A=$Particles.RibbonWidth, B=2.5)
   ) {isSetParameters=true, module=SetVariables_#3}
  ]
  particleUpdateScript [
   &SolveForcesAndVelocity(Clamp Velocity=false, Limit Acceleration=false) {enabled=false}
   &ParticleState()
   &DynamicMaterialParameters(
    Index 0 Param 1=1
    Index 0 Param 2=&FloatFromCurve(
     FloatCurve=DataInterface() {propertyValues {Curve {keys [<at 0.33 0 RCIM_Cubic>, {arriveTangent=4.6706, interpMode=RCIM_Cubic, leaveTangent=4.6706, tangentMode=RCTM_User, time=1, value=1}]}}}
     Scale Curve=-1
    )
    Index 0 Param 3=1
    Index 0 Param 4=0.1
    Module.Param0WriteEnabled=true
    Module.Param1WriteEnabled=true
    Module.Param2WriteEnabled=true
    Module.Param3WriteEnabled=true
   )
   &ScaleColor(
    Scale Alpha=&Multiply_Float(
     A=&Multiply_Float(
      A=&Multiply_Float(
       A=&Multiply_Float(
        A=&FloatFromCurve(
         FloatCurve=DataInterface() {propertyValues {Curve {keys [
             {arriveTangent=-5.8127, interpMode=RCIM_Cubic, leaveTangent=-5.8127, tangentMode=RCTM_User, time=0, value=1}
             {tangentMode=RCTM_User, time=0.35, value=0}
            ]}}}
         CurveIndex=&FN_NormalizedParticleCount()
        )
        B=&FNTimeOfDayScalar(Night Scale=0.4, Morning Scale=0.2, Evening Scale=0.4)
       )
       B=&FloatFromCurve(
        FloatCurve=DataInterface() {propertyValues {Curve {keys [
            {arriveTangent=-3.2772, interpMode=RCIM_Cubic, leaveTangent=-3.2772, tangentMode=RCTM_User, time=0, value=1}
            {arriveTangent=-3.534, interpMode=RCIM_Cubic, leaveTangent=-3.534, tangentMode=RCTM_User, time=0.375, value=0}
           ]}}}
        Scale Curve=10
       )
      )
      B=&FloatFromCurve(FloatCurve=DataInterface() {propertyValues {Curve {keys [
           {arriveTangentWeight=0.2853, interpMode=RCIM_Cubic, leaveTangentWeight=0.2853, time=0, value=0}
           {arriveTangentWeight=0.1667, interpMode=RCIM_Cubic, leaveTangentWeight=0.1667, time=0.1, value=1}
           {
            arriveTangentWeight=0.2502
            interpMode=RCIM_Cubic
            leaveTangentWeight=0.2502
            tangentMode=RCTM_User
            tangentWeightMode=RCTWM_WeightedBoth
            time=0.35
            value=0
           }
          ]}}})
     )
     B=&FloatFromCurve(
      FloatCurve=DataInterface() {propertyValues {Curve {keys [
          {interpMode=RCIM_Cubic, leaveTangentWeight=0.0333, tangentMode=RCTM_User, time=0, value=0}
          {arriveTangentWeight=0.1192, interpMode=RCIM_Cubic, leaveTangentWeight=0.228, tangentMode=RCTM_User, time=0.1, value=1}
          {arriveTangentWeight=0.0333, interpMode=RCIM_Cubic, tangentMode=RCTM_User, time=0.35, value=0}
         ]}}}
      CurveIndex=&FN_NormalizedParticleCount()
     )
    )
    Scale RGB=&VectorFromFloat(Value=&Multiply_Float(
      A=&FloatFromCurve(
       FloatCurve=DataInterface() {propertyValues {Curve {keys [
           {arriveTangent=-2.7165, interpMode=RCIM_Cubic, leaveTangent=-2.7165, tangentMode=RCTM_User, time=0.1408, value=0.9847}
           {arriveTangent=-0.0405, interpMode=RCIM_Cubic, leaveTangent=-0.0405, tangentMode=RCTM_User, time=0.5, value=0.25}
           <at 0.6 0>
          ]}}}
       CurveIndex=&FN_NormalizedParticleCount()
       Scale Curve=2
      )
      B=&FloatFromCurve(
       FloatCurve=DataInterface() {propertyValues {Curve {keys [
           {interpMode=RCIM_Cubic, tangentMode=RCTM_User, time=0, value=0}
           {interpMode=RCIM_Cubic, tangentMode=RCTM_User, time=0.5, value=1}
           {tangentMode=RCTM_User, time=1, value=0}
          ]}}}
       Scale Curve=2
      )
     ))
    ScaleA=true
    ScaleRGB=true
   ) {enabled=false}
   &ScaleColor(
    Scale Alpha=&FloatFromCurve(
     FloatCurve=DataInterface() {propertyValues {Curve {keys [
         {interpMode=RCIM_Cubic, tangentMode=RCTM_User, time=0, value=0}
         {interpMode=RCIM_Cubic, tangentMode=RCTM_User, time=0.5, value=1}
         {tangentMode=RCTM_User, time=1, value=0}
        ]}}}
     CurveIndex=&ReturnNormalizedExecIndex()
    )
    Scale RGB=&VectorFromFloat(Value=10)
    ScaleA=true
    ScaleRGB=true
   ) {module=ScaleColor001}
   /SetVariables(Particles.RibbonFacing=&ReturnMeshOrientationAxis(Mesh Orientation Axis=::ZAxis, Mesh Orientation Quat=$Particles.MeshOrientation)) {isSetParameters=true, module=SetVariables_#4}
   /SetVariables(Particles.RibbonLinkOrder:NiagaraFloat=$Particles.Age) {isSetParameters=true, module=SetVariables_#5}
   /SetVariables(Particles.ribbonWidth2Scaled:NiagaraFloat=&Multiply_Float(
     A=$Particles.Initial.ribbonWidth2
     B=&FloatFromCurve(FloatCurve=DataInterface() {propertyValues {Curve {keys [<at 0 1>,<at 1 1.125>]}}})
    )) {isSetParameters=true}
   /SetVariables(Particles.Lightning 2 color:LinearColor=&MultiplyLinearColor(
     Linear Color=$Particles.Initial.Lightning 2 color
     Scale Factor=&MakeFloat4FromVectorAndFloat(
      "Vector (RGB)"=<1 1 1>
      "Float (Alpha)"=&Multiply_Float(
       A=&FloatFromCurve(FloatCurve=DataInterface() {propertyValues {Curve {keys [
            {arriveTangent=-2.9261, interpMode=RCIM_Cubic, time=0, value=0}
            {arriveTangentWeight=0.1, interpMode=RCIM_Cubic, leaveTangentWeight=0.1667, tangentMode=RCTM_User, time=0.25, value=1}
            {arriveTangentWeight=0.2906, interpMode=RCIM_Cubic, leaveTangentWeight=0.1163, tangentMode=RCTM_User, time=0.75, value=1}
            {interpMode=RCIM_Cubic, leaveTangent=-2.6628, time=1, value=0}
           ]}}})
       B=&FloatFromCurve(
        FloatCurve=DataInterface() {propertyValues {Curve {keys [
            {arriveTangent=-2.9261, interpMode=RCIM_Cubic, time=0, value=0}
            {arriveTangentWeight=0.1, interpMode=RCIM_Cubic, leaveTangentWeight=0.1667, tangentMode=RCTM_User, time=0.25, value=1}
            {arriveTangentWeight=0.2906, interpMode=RCIM_Cubic, leaveTangentWeight=0.1163, tangentMode=RCTM_User, time=0.75, value=1}
            {interpMode=RCIM_Cubic, leaveTangent=-2.6628, time=1, value=0}
           ]}}}
        CurveIndex=&ReturnNormalizedExecIndex()
        Scale Curve=5
       )
      )
     )
    )) {enabled=false, isSetParameters=true}
  ]
  rendererClasses [&NiagaraRibbonRendererProperties]
  renderers [{rendererClass=&NiagaraRibbonRendererProperties}]
 }
}
summary {
 systemName=NS_Sienna_Male_EmoteSwing
 emitters [
  {bEnabled=false, emitterName=Lightning_pink, rendererClasses [&NiagaraRibbonRendererProperties]}
  {emitterName=Lightning_pink001, rendererClasses [&NiagaraRibbonRendererProperties]}
  {emitterName=Lightning_pink002, rendererClasses [&NiagaraRibbonRendererProperties]}
 ]
 userVariables [User.RibbonIDOffset {type=NiagaraInt32}]
}
