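# Niagara system for the Sienna male swing emote that makes short-lived magenta electric arcs crawl along the right arm.
# LargeArcs001 bursts eight-particle groups as linked ribbons, with a sprite renderer alongside.
# FN_BoneFlow places points around the User.SkeletalMesh bone using randomized angle and radius, with height driven by emitter age.
# Curves double-flash alpha and taper width over life and along each ribbon; the Empty emitter is disabled.
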NiagaraSystem /BRCosmetics/Effects/Fort_Effects/Emotes/Sienna/NS_Sienna_Male_EmoteSwing_RightArmElectricCrawl
dependencies {usedDataInterfaces [&NiagaraDataInterfaceCurve], usedDynamicInputs [], usedModules [&SystemState]}
emitters {
 Empty {
  bEnabled=false
  emitterUpdateScript [
   &EmitterState()
   &SpawnBurst_Instantaneous(Use Spawn Probability=false, UseLoopCountLimit=false)
  ]
  particleSpawnScript [&InitializeParticle(Color Mode=::NewEnumerator0, Lifetime=5, UsePositionOffset=false, Write Material Random=true)]
  particleUpdateScript [&ParticleState()]
  rendererClasses [&NiagaraComponentRendererProperties]
  renderers [{rendererClass=&NiagaraComponentRendererProperties, rendererIndex=0}]
 }
 LargeArcs001 {
  emitterUpdateScript [
   &EmitterState()
   /SetVariables(
    Emitter.random Float:NiagaraFloat=&RandomRangeFloat(Minimum=0, Maximum=1, Evaluation Type=::NewEnumerator1)
    Emitter.random height:NiagaraFloat=&RandomRangeFloat(Minimum=1.25, Maximum=1, Evaluation Type=::NewEnumerator1)
   ) {isSetParameters=true}
   &SpawnBurst_Instantaneous(Spawn Count=8, Use Spawn Probability=false, UseLoopCountLimit=false)
  ]
  particleSpawnScript [
   &FN_SetSpawnGroup(Inclusivity=true, Normalize Spawn Group=true, Particles Per Group=8, Pattern=::NewEnumerator0, Set Ribbon ID and Ribbon Link Order=true)
   &InitializeRibbon(
    Color=<rgba 6 0 6 1>
    Lifetime=&UniformRangedFloat(Minimum=0.25, Maximum=0.25)
    Ribbon Width=&RandomRangeFloat(Minimum=8, Maximum=10)
    UsePositionOffset=false
    Write Material Random=true
   )
   /SetVariables(Particles.RibbonID=/Niagara/DynamicInputs/NiagaraID/MakeNiagaraIDFromInt(Index:NiagaraInt32=$Emitter.LoopCount, Acquire Tag:NiagaraInt32=0)) {enabled=false, isSetParameters=true}
   /SetVariables(Particles.link order:NiagaraFloat=&MakeFloatFromInt(Int=&ReturnParticleIDAttributes())) {isSetParameters=true}
   &FN_BoneFlow(
    Angle:NiagaraFloat=&Add_Float(
     A=&Add_Float(
      A=&Multiply_Float(A=$Particles.Normalized Spawn Group, B=&RandomRangeFloat(Minimum=1, Maximum=1))
      B=$Emitter.random Float
     )
     B=&RandomRangeFloat(Minimum=-0.25, Maximum=0.25)
    )
    BoneID=0
    Height:NiagaraFloat=&Multiply_Float(
     A=&Lerp_Float(A=0, B=1, Alpha=&Multiply_Float(A=$Particles.Normalized Spawn Group, B=$Emitter.random height))
     B=&FloatFromCurve(
      FloatCurve=DataInterface() {propertyValues {Curve {keys [<at 0 0 RCIM_Cubic>, {arriveTangent=1.9387, interpMode=RCIM_Cubic, leaveTangent=1.9387, tangentMode=RCTM_User, time=1, value=1}]}}}
      CurveIndex=$Emitter.Age
     )
    )
    Radius:NiagaraFloat=&FloatFromCurve(
     FloatCurve=DataInterface() {propertyValues {Curve {keys [<at 0 1>,<at 1 1.5>]}}}
     CurveIndex=$Particles.Normalized Spawn Group
     Scale Curve=&RandomRangeFloat(Minimum=8, Maximum=5.5)
    )
    Set Material Dynamic=false
    SkeletalMesh:NiagaraDataInterfaceSkeletalMesh=$User.SkeletalMesh
   )
   &KillParticles(Kill Particles=&SetBoolByFloatComparison(A=$Emitter.Age, B=0.15, Comparison Type=::NewEnumerator6))
  ]
  particleUpdateScript [
   &ParticleState()
   &ScaleColor(
    Scale Alpha=&FloatFromCurve(FloatCurve=DataInterface() {propertyValues {Curve {keys [
         {arriveTangent=8.0553, interpMode=RCIM_Cubic, leaveTangent=8.0553, tangentMode=RCTM_User, time=0, value=0}
         <at 0.2113 1 RCIM_Cubic>
         <at 0.4546 0>
         {arriveTangent=-8.0018, interpMode=RCIM_Cubic, leaveTangent=7.1228, tangentMode=RCTM_Break, time=0.5517, value=0}
         <at 0.7524 1 RCIM_Cubic>
         {arriveTangent=-10.4882, interpMode=RCIM_Cubic, leaveTangent=-10.4882, tangentMode=RCTM_User, time=1, value=0}
        ]}}})
    ScaleA=true
    ScaleRGB=true
   )
   &ScaleRibbonWidth(Ribbon Width Scale=&Multiply_Float(
     A=&FloatFromCurve(FloatCurve=DataInterface() {propertyValues {Curve {keys [
          {interpMode=RCIM_Cubic, tangentMode=RCTM_User, time=0, value=0}
          {interpMode=RCIM_Cubic, tangentMode=RCTM_User, time=0.25, value=1}
          {arriveTangentWeight=0.1491, interpMode=RCIM_Cubic, leaveTangentWeight=0.2236, tangentMode=RCTM_User, time=0.75, value=1}
          {arriveTangentWeight=0.1, interpMode=RCIM_Cubic, tangentMode=RCTM_User, time=1, value=0}
         ]}}})
     B=&FloatFromCurve(
      FloatCurve=DataInterface() {propertyValues {Curve {keys [
          {arriveTangent=6.6108, interpMode=RCIM_Cubic, leaveTangent=6.6108, tangentMode=RCTM_User, time=0.05, value=0}
          {interpMode=RCIM_Cubic, tangentMode=RCTM_User, time=0.1, value=1}
          {arriveTangentWeight=0.2906, interpMode=RCIM_Cubic, leaveTangentWeight=0.1163, tangentMode=RCTM_User, time=0.9, value=1}
          {arriveTangent=-6.3941, interpMode=RCIM_Cubic, leaveTangent=-6.3941, tangentMode=RCTM_User, time=0.95, value=0}
         ]}}}
      CurveIndex=$Particles.Initial.Normalized Spawn Group
     )
    )) {module=ScaleRibbonWidth001}
  ]
  rendererClasses [&NiagaraRibbonRendererProperties, &NiagaraSpriteRendererProperties]
  renderers [
   {rendererClass=&NiagaraRibbonRendererProperties}
   {rendererClass=&NiagaraSpriteRendererProperties, rendererIndex=1}
  ]
 }
}
summary {
 systemName=NS_Sienna_Male_EmoteSwing_RightArmElectricCrawl
 emitters [
  {emitterName=LargeArcs001, rendererClasses [&NiagaraRibbonRendererProperties, &NiagaraSpriteRendererProperties]}
  {bEnabled=false, emitterName=Empty, rendererClasses [&NiagaraComponentRendererProperties]}
 ]
 userVariables [
  User.RibbonIDOffset {type=NiagaraInt32}
  User.SkeletalMesh {
   type=NiagaraDataInterfaceSkeletalMesh
   defaultValue {
    struct=NiagaraExt_VariableValue_DataInterface
    value {dataInterface=&self:NiagaraDataInterfaceSkeletalMesh_0}
   }
  }
 ]
}
