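# Niagara effect for the male Sienna emote's swing-to-pound, drawing a short-lived pink lightning ribbon trail from one continuously spawning emitter.
# Velocity and solver modules are disabled; spawn groups set ribbon IDs, and two ribbon U overrides advance with emitter age at different rates.
# Two ribbon renderers are present, alongside secondary wider, much fainter width and color attributes.
# Curves fade color and alpha along the ribbon by normalized particle count, with alpha also scaled by time of day.
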NiagaraSystem /BRCosmetics/Effects/Fort_Effects/Emotes/Sienna/NS_Sienna_Male_EmoteSwing_To_pound
dependencies {
 usedDataInterfaces [&NiagaraDataInterfaceCurve, &NiagaraDataInterfaceColorCurve]
 usedDynamicInputs []
 usedModules [&SystemState]
}
emitters {Lightning_pink {
  emitterUpdateScript [
   &EmitterState()
   &SpawnRate(SpawnRate=90, Use Spawn Probability=false)
  ]
  particleSpawnScript [
   &FN_SetSpawnGroup(Number of Spawn Groups=1, Set Ribbon ID and Ribbon Link Order=true)
   &InitializeRibbon(Color=<rgba 150 0 136.4355 25>, Lifetime=0.125, Ribbon Width=20, UsePositionOffset=false, Write Material Random=true)
   /SetVariables(Particles.RibbonU1Override:NiagaraFloat=&Add_Float(
     A=&Multiply_Float(A=$Emitter.Age, B=5)
     B=&Multiply_FloatByInt(Float=0.5, Integer=$Particles.SpawnGroup)
    )) {isSetParameters=true}
   /SetVariables(Particles.RibbonU1Override2:NiagaraFloat=&Add_Float(
     A=&Multiply_Float(A=$Emitter.Age, B=7)
     B=&Multiply_FloatByInt(Float=0.5, Integer=$Particles.SpawnGroup)
    )) {isSetParameters=true}
   &AddVelocity(Velocity=<0 200 0>, Velocity Speed Scale=5) {enabled=false}
   /SetVariables(
    Particles.Lightning 2 color:LinearColor=&MultiplyLinearColor(Linear Color=$Particles.Color, Scale Factor=<1 1 1 0.015>)
    Particles.ribbonWidth2:NiagaraFloat=&Multiply_Float(A=$Particles.RibbonWidth, B=2.5)
   ) {isSetParameters=true}
  ]
  particleUpdateScript [
   &SolveForcesAndVelocity(Clamp Velocity=false, Limit Acceleration=false) {enabled=false}
   &ParticleState()
   &DynamicMaterialParameters(
    Index 0 Param 1=1
    Index 0 Param 2=&FloatFromCurve(
     FloatCurve=DataInterface() {propertyValues {Curve {keys [<at 0.33 0 RCIM_Cubic>, {arriveTangent=4.6706, interpMode=RCIM_Cubic, leaveTangent=4.6706, tangentMode=RCTM_User, time=1, value=1}]}}}
     Scale Curve=-1
    )
    Index 0 Param 3=1
    Index 0 Param 4=0.1
    Module.Param0WriteEnabled=true
    Module.Param1WriteEnabled=true
    Module.Param2WriteEnabled=true
    Module.Param3WriteEnabled=true
   )
   &ScaleColor(
    Scale Alpha=&Multiply_Float(
     A=&Multiply_Float(
      A=&FloatFromCurve(
       FloatCurve=DataInterface() {propertyValues {Curve {keys [{arriveTangent=-11.497, time=0, value=1}, {arriveTangentWeight=0.1958, leaveTangentWeight=0.1958, time=0.2, value=0}]}}}
       CurveIndex=/Game/Effects/Niagara/DynamicInputs/Math/FN_NormalizedParticleCount()
      )
      B=&FNTimeOfDayScalar(Night Scale=0.4, Morning Scale=0.2, Evening Scale=0.4)
     )
     B=&FloatFromCurve(
      FloatCurve=DataInterface() {propertyValues {Curve {keys [
          {
           arriveTangentWeight=0.2734
           interpMode=RCIM_Cubic
           leaveTangentWeight=0.2734
           tangentMode=RCTM_User
           tangentWeightMode=RCTWM_WeightedBoth
           time=0
           value=1
          }
          {
           arriveTangentWeight=0.2345
           interpMode=RCIM_Cubic
           leaveTangentWeight=0.2345
           tangentMode=RCTM_User
           tangentWeightMode=RCTWM_WeightedBoth
           time=0.375
           value=0
          }
         ]}}}
      Scale Curve=10
     )
    )
    Scale RGB=&VectorFromFloat(Value=&FloatFromCurve(
      FloatCurve=DataInterface() {propertyValues {Curve {keys [
          {arriveTangent=-7.394, arriveTangentWeight=0.8707, leaveTangentWeight=0.8707, time=0, value=0.9847}
          {
           arriveTangentWeight=0.2307
           interpMode=RCIM_Cubic
           leaveTangentWeight=0.2307
           tangentMode=RCTM_User
           tangentWeightMode=RCTWM_WeightedBoth
           time=0.35
           value=0
          }
         ]}}}
      CurveIndex=&FN_NormalizedParticleCount()
      Scale Curve=2
     ))
    ScaleA=true
    ScaleRGB=true
   )
   /SetVariables(Particles.RibbonFacing=&ReturnMeshOrientationAxis(Mesh Orientation Axis=::ZAxis, Mesh Orientation Quat=$Particles.MeshOrientation)) {isSetParameters=true}
   /SetVariables(Particles.RibbonLinkOrder:NiagaraFloat=$Particles.Age) {isSetParameters=true}
   /SetVariables(Particles.ribbonWidth2Scaled:NiagaraFloat=&Multiply_Float(
     A=$Particles.Initial.ribbonWidth2
     B=&FloatFromCurve(FloatCurve=DataInterface() {propertyValues {Curve {keys [<at 0 1>,<at 1 1.125>]}}})
    )) {isSetParameters=true}
  ]
  rendererClasses [&NiagaraRibbonRendererProperties]
  renderers [
   {rendererClass=&NiagaraRibbonRendererProperties}
   {rendererClass=&NiagaraRibbonRendererProperties, rendererIndex=1}
  ]
 }}
summary {
 systemName=NS_Sienna_Male_EmoteSwing_To_pound
 emitters [{emitterName=Lightning_pink, rendererClasses [&NiagaraRibbonRendererProperties]}]
 userVariables [User.RibbonIDOffset {type=NiagaraInt32}]
}
