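# Niagara effect for the male Sienna emote's arm swing, drawing short-lived magenta arc ribbons.
# Its single emitter bursts eight particles, linked into one ribbon per emitter loop by a RibbonID taken from LoopCount.
# Cylinder placement driven by normalized spawn index, plus random offsets, lays each ribbon out as a jittered arc along the cylinder.
# Alpha double-pulses over lifetime while width fades in and out and tapers at both ends; ribbon and sprite renderers draw it.
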NiagaraSystem /BRCosmetics/Effects/Fort_Effects/Emotes/Sienna/NS_Sienna_Male_EmoteSwing_arm
dependencies {usedDataInterfaces [&NiagaraDataInterfaceCurve], usedDynamicInputs [], usedModules [&SystemState]}
emitters {LargeArcs001 {
  emitterUpdateScript [
   &EmitterState()
   &SpawnBurst_Instantaneous(Spawn Count=8, Use Spawn Probability=false, UseLoopCountLimit=false)
   /SetVariables(Emitter.random Float:NiagaraFloat=&RandomRangeFloat(Minimum=0, Maximum=1)) {isSetParameters=true}
  ]
  particleSpawnScript [
   /Game/Effects/Niagara/Modules/Partition/FN_ParticlePartition_v2/FN_Particle_Partition(Enable Ribbon ID=true, Seed:NiagaraInt32=$Emitter.Seed, Shuffle Partition ID=true) {enabled=false}
   &InitializeRibbon(
    Color=<rgba 6 0 6 1>
    Lifetime=&UniformRangedFloat(Minimum=0.2, Maximum=0.35)
    Module.Write Color=true
    Module.Write Lifetime=true
    Module.Write Mass=true
    Module.Write Material Random=true
    Module.Write Position=false
    Module.Write RibbonTwist=false
    Module.Write RibbonWidth=true
    Ribbon Width=10
   )
   /SetVariables(
    Particles.RibbonID=/Niagara/DynamicInputs/NiagaraID/MakeNiagaraIDFromInt(Index:NiagaraInt32=$Emitter.LoopCount, Acquire Tag:NiagaraInt32=0)
    Particles.link order:NiagaraFloat=&MakeFloatFromInt(Int=&ReturnParticleIDAttributes())
   ) {isSetParameters=true}
   &CylinderLocation(
    Clamp UV Range=false
    Cylinder Distribution:ENiagara_CylinderMode=::NewEnumerator2
    Cylinder Height=40
    Cylinder Height Midpoint=0
    Cylinder Origin:NiagaraPosition=/Niagara/DynamicInputs/Helpers/SimulationPosition()
    Cylinder Radius=&RandomRangeFloat(Minimum=5, Maximum=7)
    MaskSpawnGroup=false
    Radius Position:NiagaraFloat=&FloatFromCurve(
     FloatCurve=DataInterface() {propertyValues {Curve {keys [
         {
          arriveTangent=3.759
          arriveTangentWeight=0.1067
          interpMode=RCIM_Cubic
          leaveTangent=3.759
          leaveTangentWeight=0.3202
          tangentMode=RCTM_User
          time=0
          value=0
         }
         {interpMode=RCIM_Cubic, tangentMode=RCTM_User, time=0.8, value=1}
        ]}}}
     CurveIndex=&ReturnNormalizedExecIndex()
    )
    Rotation Mode:ENiagara_RotationMode=::NewEnumerator1
    U Position:NiagaraFloat=&Add_Float(
     A=&Add_Float(A=&Add_Float(A=&ReturnNormalizedExecIndex(), B=0.35), B=$Emitter.random Float)
     B=&RandomRangeFloat(Minimum=-0.2, Maximum=0.2)
    )
    UseOwnerScale=true
    V Position:NiagaraFloat=&ReturnNormalizedExecIndex()
   )
  ]
  particleUpdateScript [
   &ParticleState()
   &ScaleColor(
    Scale Alpha=&FloatFromCurve(FloatCurve=DataInterface() {propertyValues {Curve {keys [
         {arriveTangent=8.0553, interpMode=RCIM_Cubic, leaveTangent=8.0553, tangentMode=RCTM_User, time=0, value=0}
         <at 0.2113 1 RCIM_Cubic>
         <at 0.4546 0>
         {arriveTangent=-8.0018, interpMode=RCIM_Cubic, leaveTangent=7.1228, tangentMode=RCTM_Break, time=0.5517, value=0}
         <at 0.7524 1 RCIM_Cubic>
         {arriveTangent=-10.4882, interpMode=RCIM_Cubic, leaveTangent=-10.4882, tangentMode=RCTM_User, time=1, value=0}
        ]}}})
    ScaleA=true
    ScaleRGB=true
   )
   &ScaleRibbonWidth(Ribbon Width Scale=&Multiply_Float(
     A=&FloatFromCurve(FloatCurve=DataInterface() {propertyValues {Curve {keys [
          {interpMode=RCIM_Cubic, tangentMode=RCTM_User, time=0, value=0}
          {interpMode=RCIM_Cubic, tangentMode=RCTM_User, time=0.25, value=1}
          {arriveTangentWeight=0.1491, interpMode=RCIM_Cubic, leaveTangentWeight=0.2236, tangentMode=RCTM_User, time=0.75, value=1}
          {arriveTangentWeight=0.1, interpMode=RCIM_Cubic, tangentMode=RCTM_User, time=1, value=0}
         ]}}})
     B=&FloatFromCurve(
      FloatCurve=DataInterface() {propertyValues {Curve {keys [
          {arriveTangent=6.6108, interpMode=RCIM_Cubic, leaveTangent=6.6108, tangentMode=RCTM_User, time=0.05, value=0}
          {interpMode=RCIM_Cubic, tangentMode=RCTM_User, time=0.5, value=1}
          {arriveTangent=-6.3941, interpMode=RCIM_Cubic, leaveTangent=-6.3941, tangentMode=RCTM_User, time=0.95, value=0}
         ]}}}
      CurveIndex=&ReturnNormalizedExecIndex()
     )
    )) {module=ScaleRibbonWidth001}
  ]
  rendererClasses [&NiagaraRibbonRendererProperties, &NiagaraSpriteRendererProperties]
  renderers [
   {rendererClass=&NiagaraRibbonRendererProperties}
   {rendererClass=&NiagaraSpriteRendererProperties, rendererIndex=1}
  ]
 }}
summary {
 systemName=NS_Sienna_Male_EmoteSwing_arm
 emitters [{emitterName=LargeArcs001, rendererClasses [&NiagaraRibbonRendererProperties, &NiagaraSpriteRendererProperties]}]
 userVariables [User.RibbonIDOffset {type=NiagaraInt32}]
}
