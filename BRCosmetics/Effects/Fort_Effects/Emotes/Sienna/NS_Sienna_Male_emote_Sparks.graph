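# Niagara system for a spark spray, named for the male variant of the Sienna emote.
# A single sprite emitter continuously spawns short-lived orange sparks from a small, non-uniformly scaled shape and launches them quickly in a narrow cone.
# Heavy drag, curl noise and strong gravity break up the motion, while curves fade alpha, RGB is boosted, and size is randomized and stretched by speed.
# A hand_l skeletal mesh location module is present but disabled.
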NiagaraSystem /BRCosmetics/Effects/Fort_Effects/Emotes/Sienna/NS_Sienna_Male_emote_Sparks
dependencies {
 usedDataInterfaces [
  &NiagaraDataInterfaceSkeletalMesh
  &NiagaraDataInterfaceCurve
  &NiagaraDataInterfaceVector2DCurve
 ]
 usedDynamicInputs [/Game/Effects/Niagara/DynamicInputs/Math/FN_Fit]
 usedModules [
  &self:SystemSpawnScript.SystemScriptSource.SystemScriptGraph.NiagaraNodeAssignment_0.SetVariables_#1
  &SystemState
 ]
}
emitters {Sparkles {
  emitterUpdateScript [
   &EmitterState()
   &SpawnRate(SpawnRate=90, Use Spawn Probability=false)
  ]
  particleSpawnScript [
   &InitializeParticle(
    Color=<rgba 1 0.4505 0.026 1>
    Lifetime=0.15
    Sprite Size=<5 10>
    Sprite Size Mode=::NewEnumerator4
    UsePositionOffset=false
    Write Material Random=true
   ) {module=InitializeParticle001}
   &SkeletalMeshLocation(
    Bone Sampling Mode=::NewEnumerator4
    MaskSpawnGroup=false
    Skeletal Mesh=DataInterface() {propertyValues {
      DefaultMesh=null
      ExcludeBoneName=root
      OptionalComponentClass=null
      PreviewMesh=/BRCosmetics/Characters/Player/Male/Medium/Bodies/M_MED_Sienna/Meshes/M_MED_Sienna
      SkinningMode=SkinOnTheFly
      SoftSourceActor=null
      SourceMode=Default
      UvSetIndex=0
      WholeMeshLOD=-1
      bExcludeBone=false
      bReadDeformedGeometry=true
      bRequireCurrentFrameData=true
      ComponentTags []
      FilteredBones [hand_l]
      FilteredSockets []
      MeshUserParameter {parameter {name=None, typeDefHandle {registeredTypeIndex=1}, varData []}}
      SamplingRegions []
     }}
   ) {enabled=false}
   &ShapeLocation(MaskSpawnGroup=false, Non Uniform Scale=<1 5 0.1>, Override Seed=false, Sphere Radius=1, UseOwnerScale=true)
   &AddVelocity(
    Cone Angle=15
    Module.UseConeFalloff=false
    Module.UseVelDistribution=false
    Override Seed=false
    Velocity Mode=::NewEnumerator2
    Velocity Speed=1000
   )
  ]
  particleUpdateScript [
   &ParticleState()
   &Drag(Drag=25, Use Linear Drag=true, Use Rotational Drag=false)
   &CurlNoiseForce(Noise Quality / Cost=::NewEnumerator0, Noise Strength=1535.5676, Pan Noise=false)
   &GravityForce(Gravity=<0 0 -2940>)
   &SolveForcesAndVelocity(Clamp Velocity=false, Limit Acceleration=false)
   &ScaleColor(
    Scale Alpha=&FloatFromCurve(FloatCurve=DataInterface() {propertyValues {Curve {keys [<at -0.0092 0.0144>,<at 0.0967 0.9767>,<at 1.0011 0.002>]}}})
    Scale RGB=<15 5 1>
    ScaleA=true
    ScaleRGB=true
   )
   &ScaleSpriteSize(
    Scale Factor=&MakeVector2D(
     X=0.5
     Y=&FloatFromCurve(
      FloatCurve=DataInterface() {propertyValues {Curve {keys [<at -0.0046 0>,<at 0.0873 0.9701>,<at 1.0011 0.002>]}}}
      CurveIndex=&RandomRangeFloat(Minimum=0.5, Maximum=1)
     )
    )
    Scale Sprite Size Mode=::NewEnumerator2
   )
   &ScaleSpriteSizeBySpeed(Max Scale Factor=<1 3>, Min Scale Factor=<1 1.5>, Velocity Threshold=250)
  ]
  rendererClasses [&NiagaraSpriteRendererProperties]
  renderers [{rendererClass=&NiagaraSpriteRendererProperties}]
 }}
summary {
 systemName=NS_Sienna_Male_emote_Sparks
 emitters [{emitterName=Sparkles, rendererClasses [&NiagaraSpriteRendererProperties]}]
 userVariables []
}
